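# Physics asset for the long pencil skirt lower-body mesh on the female medium body, simulating skirt bone chains with body collision.
# Capsules on pelvis, thighs, calves and feet join dyn_Cloth_volume chains around front, back and sides, each rooted to the pelvis by a tiny capsule.
# Front and A chains swing widely in one axis but narrowly in the other; some B and C bones add angled capsules.
# An unconstrained dyn_Floor_COL body supplies large-sphere floor collision.
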PhysicsAsset /BRCosmetics/Characters/Player/Female/Medium/Bodies/F_MED_Pencil_B/SkirtLong/Meshes/Pencil_SkirtLong_MedB_Lbdy_Physics
bodies {
 calf_l [{shapeType=Capsule, length=51.1602, radius=8.1734, center=<-35.6612 0.0526 -0.1273>, rotation=<rot 0.1246 84.5345 -88.5757>}]
 calf_r [{shapeType=Capsule, length=52.4102, radius=8.1734, center=<34.3144 0.6477 -0.4558>, rotation=<rot -0.3385 89.5329 -88.4634>}]
 dyn_Cloth_volume_bk [{shapeType=Capsule, length=0.505, radius=0.505, shapeName=dyn_Cloth_volume_bk_capsule, center=<-8 -8 0>}]
 dyn_Cloth_volume_bk01 [{shapeType=Capsule, length=14.25, radius=8.5, shapeName=dyn_Cloth_volume_bk01_capsule, center=<-12 -8 0>}]
 dyn_Cloth_volume_bk02 [{shapeType=Capsule, length=24.875, radius=10, shapeName=dyn_Cloth_volume_bk02_capsule, center=<-8 -8 0>}]
 dyn_Cloth_volume_bk03 [{shapeType=Capsule, length=35, radius=8.75, shapeName=dyn_Cloth_volume_bk03_capsule, center=<-8 -8 0>}]
 dyn_Cloth_volume_fn01 [{shapeType=Capsule, length=0.505, radius=0.505, shapeName=dyn_Cloth_volume_bk_capsule, center=<-8 -8 0>}]
 dyn_Cloth_volume_fn02 [{shapeType=Capsule, length=22.75, radius=9.625, shapeName=dyn_Cloth_volume_bk02_capsule, center=<-8 -6 0>}]
 dyn_Cloth_volume_fn03 [{shapeType=Capsule, length=28, radius=9.875, shapeName=dyn_Cloth_volume_bk03_capsule, center=<-8 -6 0>}]
 dyn_Cloth_volume_l_A01 [{shapeType=Capsule, length=0.505, radius=0.505, shapeName=dyn_Cloth_volume_l_C_capsule, center=<-8 -8 0>}]
 dyn_Cloth_volume_l_A02 [{shapeType=Capsule, length=18.5, radius=9.625, shapeName=dyn_Cloth_volume_bk02_capsule, center=<-8 -8 3>, rotation=<rot 0.0001 0.0002 5>}]
 dyn_Cloth_volume_l_A03 [{shapeType=Capsule, length=24.625, radius=9.125, shapeName=dyn_Cloth_volume_bk03_capsule, center=<-8 -8 3>, rotation=<rot 0.0001 0.0002 5>}]
 dyn_Cloth_volume_l_B [{shapeType=Capsule, length=0.505, radius=0.505, shapeName=dyn_Cloth_volume_bk_capsule, center=<-8 -8 0>}]
 dyn_Cloth_volume_l_B01 [{shapeType=Capsule, length=16.875, radius=8.5, shapeName=dyn_Cloth_volume_bk01_capsule, center=<-12 -13 10>}]
 dyn_Cloth_volume_l_B02 [
  {shapeType=Capsule, length=25.25, radius=10.625, shapeName=dyn_Cloth_volume_bk02_capsule, center=<-8 -13 10>, rotation=<rot 0.0001 0 10>}
  {
   shapeType=Capsule
   length=25.375
   radius=11.125
   shapeName=dyn_Cloth_volume_l_B02_capsule
   center=<-8 -13.2854 -3.8047>
   rotation=<rot 0.0009 0 -25>
  }
 ]
 dyn_Cloth_volume_l_B03 [
  {shapeType=Capsule, length=31.125, radius=10.375, shapeName=dyn_Cloth_volume_bk03_capsule, center=<-8 -13 10>, rotation=<rot 0.0001 0 10>}
  {
   shapeType=Capsule
   length=25.375
   radius=11.125
   shapeName=dyn_Cloth_volume_l_B03_capsule
   center=<-8 -13.1568 -4.0067>
   rotation=<rot 0.0005 0 -25>
  }
 ]
 dyn_Cloth_volume_l_C [{shapeType=Capsule, length=0.505, radius=0.505, shapeName=dyn_Cloth_volume_l_C_capsule, center=<-8 -8 0>}]
 dyn_Cloth_volume_l_C01 [{shapeType=Capsule, length=16.875, radius=8.5, shapeName=dyn_Cloth_volume_bk01_capsule, center=<-12 -8 0>}]
 dyn_Cloth_volume_l_C02 [{shapeType=Capsule, length=25.25, radius=10.75, shapeName=dyn_Cloth_volume_bk02_capsule, center=<-8 -8 0>}]
 dyn_Cloth_volume_l_C03 [{shapeType=Capsule, length=32.125, radius=9.5, shapeName=dyn_Cloth_volume_bk03_capsule, center=<-8 -8 0>}]
 dyn_Cloth_volume_r_A01 [{shapeType=Capsule, length=0.505, radius=0.505, shapeName=dyn_Cloth_volume_l_C_capsule, center=<-8 -8 0>}]
 dyn_Cloth_volume_r_A02 [{shapeType=Capsule, length=26, radius=8.5, shapeName=dyn_Cloth_volume_bk02_capsule, center=<-8 -8 -3>, rotation=<rot -0.0001 0.0003 -5>}]
 dyn_Cloth_volume_r_A03 [{shapeType=Capsule, length=30, radius=10.125, shapeName=dyn_Cloth_volume_bk03_capsule, center=<-8 -8 -3>, rotation=<rot -0.0001 0.0004 -5>}]
 dyn_Cloth_volume_r_B [{shapeType=Capsule, length=0.505, radius=0.505, shapeName=dyn_Cloth_volume_bk_capsule, center=<-8 -8 0>}]
 dyn_Cloth_volume_r_B01 [{shapeType=Capsule, length=24.625, radius=8.375, shapeName=dyn_Cloth_volume_bk01_capsule, center=<-12 -13 -10>, rotation=<rot -0.0001 0 -15>}]
 dyn_Cloth_volume_r_B02 [
  {shapeType=Capsule, length=27.875, radius=10.375, shapeName=dyn_Cloth_volume_bk02_capsule, center=<-8 -12 -10>, rotation=<rot -0.0003 0 -20>}
  {
   shapeType=Capsule
   length=23.75
   radius=9.875
   shapeName=dyn_Cloth_volume_r_B02_capsule
   center=<-8 -11.4821 5.9201>
   rotation=<rot -0.0004 0 30>
  }
 ]
 dyn_Cloth_volume_r_B03 [
  {shapeType=Capsule, length=17.25, radius=11.375, shapeName=dyn_Cloth_volume_bk03_capsule, center=<-8 -12 -10>, rotation=<rot -0.0003 0 -20>}
  {
   shapeType=Capsule
   length=23.75
   radius=9.875
   shapeName=dyn_Cloth_volume_r_B03_capsule
   center=<-8 -11.4821 5.9201>
   rotation=<rot -0.0004 0 30>
  }
  {
   shapeType=Capsule
   length=8.625
   radius=10
   shapeName=dyn_Cloth_volume_r_B03_capsule
   center=<-7.8207 -2.1597 -22.3314>
   rotation=<rot -0.0002 0 -60>
  }
 ]
 dyn_Cloth_volume_r_C [{shapeType=Capsule, length=0.505, radius=0.505, shapeName=dyn_Cloth_volume_r_C_capsule, center=<-8 -8 0>}]
 dyn_Cloth_volume_r_C01 [{shapeType=Capsule, length=24.8376, radius=8.7874, shapeName=dyn_Cloth_volume_bk01_capsule, center=<-12 -8 4>}]
 dyn_Cloth_volume_r_C02 [
  {shapeType=Capsule, length=28.5876, radius=11.5374, shapeName=dyn_Cloth_volume_bk02_capsule, center=<-8 -8 4>, rotation=<rot 0 0 5>}
  {
   shapeType=Capsule
   length=19.875
   radius=10.125
   shapeName=dyn_Cloth_volume_r_C02_capsule
   center=<-6.1155 -3.3633 15.3171>
   rotation=<rot 0 -0.0002 45>
  }
 ]
 dyn_Cloth_volume_r_C03 [
  {shapeType=Capsule, length=30.9626, radius=10.7874, shapeName=dyn_Cloth_volume_bk03_capsule, center=<-8 -8 0>, rotation=<rot 0 0 5>}
  {
   shapeType=Capsule
   length=19.875
   radius=10.125
   shapeName=dyn_Cloth_volume_r_C03_capsule
   center=<-6.1155 -3.3633 15.3171>
   rotation=<rot 0 -0.0002 45>
  }
 ]
 dyn_Floor_COL [
  {shapeType=Sphere, radius=198.875, shapeName=floorCOL_sphere, center=<-3.4923 -2.714 -192.0627>}
  {shapeType=Sphere, radius=27.3335, shapeName=floorCOL_sphere, center=<0 0.5113 -6.1886>}
 ]
 foot_l [{
   shapeType=Capsule
   length=7.1162
   radius=12.3129
   shapeName=foot_l_capsule
   center=<-5.043 4.5841 0.5164>
   rotation=<rot 2.2652 -176.811 -85.1339>
  }]
 foot_r [{
   shapeType=Capsule
   length=8.4912
   radius=10.9379
   shapeName=foot_r_capsule
   center=<3.6724 -3.9727 -0.3857>
   rotation=<rot 2.0111 4.5862 -84.4638>
  }]
 pelvis [
  {shapeType=Capsule, length=7.7125, radius=8.4301, center=<-0.9227 -1.2812 0>}
  {shapeType=Capsule, length=10.79, radius=7.4702, shapeName=pelvis_capsule, center=<-3.9978 -3.5061 0>}
 ]
 thigh_l [{shapeType=Capsule, length=29.75, radius=7.75, center=<-21.6167 -1.0761 1.3426>, rotation=<rot 90 89.4328 89.6284>}]
 thigh_r [{shapeType=Capsule, length=29.625, radius=7.875, center=<21.0533 0.5413 -1.1849>, rotation=<rot 90 89.4327 89.4328>}]
}
constraints [
 {bone1Name=thigh_l, bone2Name=pelvis, swing1LimitDegrees=45, swing2LimitDegrees=45, twistLimitDegrees=45}
 {bone1Name=thigh_r, bone2Name=pelvis, swing1LimitDegrees=45, swing2LimitDegrees=45, twistLimitDegrees=45}
 {bone1Name=dyn_Cloth_volume_r_C02, bone2Name=dyn_Cloth_volume_r_C01, swing1LimitDegrees=45, swing2LimitDegrees=5, twistLimitDegrees=15}
 {bone1Name=dyn_Cloth_volume_l_C02, bone2Name=dyn_Cloth_volume_l_C01, swing1LimitDegrees=55, swing2LimitDegrees=5, twistLimitDegrees=15}
 {
  bone1Name=calf_r
  bone2Name=thigh_r
  swing1LimitDegrees=45
  swing1Motion=Free
  swing2LimitDegrees=45
  swing2Motion=Free
  twistLimitDegrees=45
  twistMotion=Free
 }
 {
  bone1Name=calf_l
  bone2Name=thigh_l
  swing1LimitDegrees=45
  swing1Motion=Free
  swing2LimitDegrees=45
  swing2Motion=Free
  twistLimitDegrees=45
  twistMotion=Free
 }
 {bone1Name=dyn_Cloth_volume_l_C03, bone2Name=dyn_Cloth_volume_l_C02, swing1LimitDegrees=55, swing2LimitDegrees=5, twistLimitDegrees=15}
 {bone1Name=dyn_Cloth_volume_r_C03, bone2Name=dyn_Cloth_volume_r_C02, swing1LimitDegrees=45, swing2LimitDegrees=5, twistLimitDegrees=15}
 {bone1Name=dyn_Cloth_volume_r_C01, bone2Name=dyn_Cloth_volume_r_C, swing1LimitDegrees=45, swing2LimitDegrees=5, twistLimitDegrees=15}
 {bone1Name=dyn_Cloth_volume_l_C01, bone2Name=dyn_Cloth_volume_l_C, swing1LimitDegrees=55, swing2LimitDegrees=5, twistLimitDegrees=15}
 {bone1Name=dyn_Cloth_volume_bk01, bone2Name=dyn_Cloth_volume_bk, swing1LimitDegrees=45, swing2LimitDegrees=5, twistLimitDegrees=15}
 {bone1Name=dyn_Cloth_volume_bk02, bone2Name=dyn_Cloth_volume_bk01, swing1LimitDegrees=45, swing2LimitDegrees=5, twistLimitDegrees=15}
 {bone1Name=dyn_Cloth_volume_bk03, bone2Name=dyn_Cloth_volume_bk02, swing1LimitDegrees=45, swing2LimitDegrees=5, twistLimitDegrees=15}
 {bone1Name=dyn_Cloth_volume_r_C, bone2Name=pelvis, swing1LimitDegrees=45, swing2LimitDegrees=45, twistLimitDegrees=15}
 {bone1Name=dyn_Cloth_volume_l_C, bone2Name=pelvis, swing1LimitDegrees=45, swing2LimitDegrees=45, twistLimitDegrees=15}
 {bone1Name=dyn_Cloth_volume_bk, bone2Name=pelvis, swing1LimitDegrees=45, swing2LimitDegrees=45, twistLimitDegrees=15}
 {
  bone1Name=foot_r
  bone2Name=calf_r
  swing1LimitDegrees=45
  swing1Motion=Free
  swing2LimitDegrees=45
  swing2Motion=Free
  twistLimitDegrees=45
  twistMotion=Free
 }
 {
  bone1Name=foot_l
  bone2Name=calf_l
  swing1LimitDegrees=45
  swing1Motion=Free
  swing2LimitDegrees=45
  swing2Motion=Free
  twistLimitDegrees=45
  twistMotion=Free
 }
 {
  bone1Name=dyn_Cloth_volume_r_B
  bone2Name=pelvis
  swing1LimitDegrees=45
  swing1Motion=Free
  swing2LimitDegrees=45
  swing2Motion=Free
  twistLimitDegrees=45
  twistMotion=Free
 }
 {bone1Name=dyn_Cloth_volume_l_B03, bone2Name=dyn_Cloth_volume_l_B02, swing1LimitDegrees=45, swing2LimitDegrees=45, twistLimitDegrees=25}
 {
  bone1Name=dyn_Cloth_volume_l_B
  bone2Name=pelvis
  swing1LimitDegrees=45
  swing1Motion=Free
  swing2LimitDegrees=45
  swing2Motion=Free
  twistLimitDegrees=45
  twistMotion=Free
 }
 {bone1Name=dyn_Cloth_volume_l_B01, bone2Name=dyn_Cloth_volume_l_B, swing1LimitDegrees=45, swing2LimitDegrees=45, twistLimitDegrees=25}
 {bone1Name=dyn_Cloth_volume_l_B02, bone2Name=dyn_Cloth_volume_l_B01, swing1LimitDegrees=45, swing2LimitDegrees=45, twistLimitDegrees=25}
 {bone1Name=dyn_Cloth_volume_r_B01, bone2Name=dyn_Cloth_volume_r_B, swing1LimitDegrees=45, swing2LimitDegrees=45, twistLimitDegrees=25}
 {bone1Name=dyn_Cloth_volume_r_B02, bone2Name=dyn_Cloth_volume_r_B01, swing1LimitDegrees=45, swing2LimitDegrees=45, twistLimitDegrees=25}
 {bone1Name=dyn_Cloth_volume_r_B03, bone2Name=dyn_Cloth_volume_r_B02, swing1LimitDegrees=45, swing2LimitDegrees=45, twistLimitDegrees=25}
 {
  bone1Name=dyn_Cloth_volume_l_A01
  bone2Name=pelvis
  swing1LimitDegrees=45
  swing1Motion=Free
  swing2LimitDegrees=45
  swing2Motion=Free
  twistLimitDegrees=45
  twistMotion=Free
 }
 {bone1Name=dyn_Cloth_volume_l_A02, bone2Name=dyn_Cloth_volume_l_A01, swing1LimitDegrees=100, swing2LimitDegrees=5, twistLimitDegrees=15}
 {bone1Name=dyn_Cloth_volume_l_A03, bone2Name=dyn_Cloth_volume_l_A02, swing1LimitDegrees=100, swing2LimitDegrees=5, twistLimitDegrees=15}
 {
  bone1Name=dyn_Cloth_volume_r_A01
  bone2Name=pelvis
  swing1LimitDegrees=45
  swing1Motion=Free
  swing2LimitDegrees=45
  swing2Motion=Free
  twistLimitDegrees=45
  twistMotion=Free
 }
 {bone1Name=dyn_Cloth_volume_r_A02, bone2Name=dyn_Cloth_volume_r_A01, swing1LimitDegrees=100, swing2LimitDegrees=5, twistLimitDegrees=15}
 {bone1Name=dyn_Cloth_volume_r_A03, bone2Name=dyn_Cloth_volume_r_A02, swing1LimitDegrees=100, swing2LimitDegrees=5, twistLimitDegrees=15}
 {
  bone1Name=dyn_Cloth_volume_fn01
  bone2Name=pelvis
  swing1LimitDegrees=45
  swing1Motion=Free
  swing2LimitDegrees=45
  swing2Motion=Free
  twistLimitDegrees=45
  twistMotion=Free
 }
 {bone1Name=dyn_Cloth_volume_fn02, bone2Name=dyn_Cloth_volume_fn01, swing1LimitDegrees=100, swing2LimitDegrees=10, twistLimitDegrees=25}
 {bone1Name=dyn_Cloth_volume_fn03, bone2Name=dyn_Cloth_volume_fn02, swing1LimitDegrees=100, swing2LimitDegrees=10, twistLimitDegrees=25}
]
