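# Animation Blueprint for a female medium-body long pencil skirt cosmetic, shaping and simulating its cloth bones on the character.
# It copies the attached parent's pose through a DeformRig, then layers additive bone offsets and front-bone constraints for movement, frontend and traversal states.
# A pelvis-space RigidBody node with overridden gravity follows, then strafe and state corrections and a post Control Rig.
# The EventGraph sets the driving alphas; some constraint, parachute and Control Rig nodes are unconnected.
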AnimBlueprint /BRCosmetics/Characters/Player/Female/Medium/Bodies/F_MED_Pencil_B/SkirtLong/Meshes/Pencil_SkirtLong_MedB_Lbdy_AnimBP
target_skeleton=&/Pencil_SkirtLong_MedB_Lbdy_Skeleton
graphs [
 AnimGraph {nodes [
   n1 = "Constraint - dyn_Cloth_volume_r_C (PPP)"!(ComponentPose, Alpha=1, ConstraintWeights_0=0.3367, ConstraintWeights_1:float=0.256, ConstraintWeights_2:float=0.4073) -> (Pose) {
    alphaBoolBlend {blendOption=Linear}
    boneToModify {boneName=dyn_Cloth_volume_r_C}
    constraintSetup [{targetBone {boneName=pelvis}}, {targetBone {boneName=thigh_twist_01_r}}, {targetBone {boneName=deform_glute_r}}]
    constraintWeights [1,1,1]
   }
   n2 = "Constraint - dyn_Cloth_volume_bk (PPPPP)"!(
    ComponentPose=@n1
    Alpha=1
    ConstraintWeights_0=0.7511
    ConstraintWeights_1:float=0.0501
    ConstraintWeights_2:float=0.0504
    ConstraintWeights_3:float=0.0735
    ConstraintWeights_4:float=0.0749
   ) -> (Pose) {
    alphaBoolBlend {blendOption=Linear}
    boneToModify {boneName=dyn_Cloth_volume_bk}
    constraintSetup [
     {targetBone {boneName=pelvis}}
     {targetBone {boneName=thigh_twist_01_l}}
     {targetBone {boneName=thigh_twist_01_r}}
     {targetBone {boneName=deform_glute_l}}
     {targetBone {boneName=deform_glute_r}}
    ]
    constraintWeights [1,1,1,1,1]
   }
   n3 = "Constraint - dyn_Cloth_volume_l_C (PPP)"!(ComponentPose=@n2, Alpha=1, ConstraintWeights_0=0.3367, ConstraintWeights_1:float=0.256, ConstraintWeights_2:float=0.4073) -> (Pose) {
    alphaBoolBlend {blendOption=Linear}
    boneToModify {boneName=dyn_Cloth_volume_l_C}
    constraintSetup [{targetBone {boneName=pelvis}}, {targetBone {boneName=thigh_twist_01_l}}, {targetBone {boneName=deform_glute_l}}]
    constraintWeights [1,1,1]
   }
   n4 = "Constraint - dyn_Cloth_volume_r_B (PPPPP)"!(
    ComponentPose=@n3
    Alpha=1
    ConstraintWeights_0=0.1783
    ConstraintWeights_1:float=0.6512
    ConstraintWeights_2:float=0.1127
    ConstraintWeights_3:float=0.0301
    ConstraintWeights_4:float=0.0276
   ) -> (Pose) {
    alphaBoolBlend {blendOption=Linear}
    boneToModify {boneName=dyn_Cloth_volume_r_B}
    constraintSetup [
     {targetBone {boneName=pelvis}}
     {targetBone {boneName=thigh_twist_01_r}}
     {targetBone {boneName=thigh_r}}
     {targetBone {boneName=deform_glute_r}}
     {targetBone {boneName=deform_groin_r}}
    ]
    constraintWeights [1,1,1,1,1]
   }
   n5 = "Constraint - dyn_Cloth_volume_l_B (PPPPP)"!(
    ComponentPose=@n4
    Alpha=1
    ConstraintWeights_0=0.1783
    ConstraintWeights_1:float=0.6512
    ConstraintWeights_2:float=0.1127
    ConstraintWeights_3:float=0.0301
    ConstraintWeights_4:float=0.0276
   ) -> (Pose) {
    alphaBoolBlend {blendOption=Linear}
    boneToModify {boneName=dyn_Cloth_volume_l_B}
    constraintSetup [
     {targetBone {boneName=pelvis}}
     {targetBone {boneName=thigh_twist_01_l}}
     {targetBone {boneName=thigh_l}}
     {targetBone {boneName=deform_glute_l}}
     {targetBone {boneName=deform_groin_l}}
    ]
    constraintWeights [1,1,1,1,1]
   }
   n6 = "Constraint - dyn_Cloth_volume_r_A (PPPP)"!(
    ComponentPose=@n5
    Alpha=1
    ConstraintWeights_0=0.2372
    ConstraintWeights_1:float=0.4317
    ConstraintWeights_2:float=0.0623
    ConstraintWeights_3:float=0.2688
   ) -> (Pose) {
    alphaBoolBlend {blendOption=Linear}
    boneToModify {boneName=dyn_Cloth_volume_r_A}
    constraintSetup [
     {targetBone {boneName=pelvis}}
     {targetBone {boneName=thigh_twist_01_r}}
     {targetBone {boneName=thigh_r}}
     {targetBone {boneName=deform_groin_r}}
    ]
    constraintWeights [1,1,1,1]
   }
   n7 = "Constraint - dyn_Cloth_volume_fn (PPPPP)"!(
    ComponentPose=@n6
    Alpha=1
    ConstraintWeights_0=0.5182
    ConstraintWeights_1:float=0.1769
    ConstraintWeights_2:float=0.1714
    ConstraintWeights_3:float=0.0703
    ConstraintWeights_4:float=0.0631
   ) -> (Pose) {
    alphaBoolBlend {blendOption=Linear}
    boneToModify {boneName=dyn_Cloth_volume_fn}
    constraintSetup [
     {targetBone {boneName=pelvis}}
     {targetBone {boneName=thigh_twist_01_l}}
     {targetBone {boneName=thigh_twist_01_r}}
     {targetBone {boneName=deform_groin_l}}
     {targetBone {boneName=deform_groin_r}}
    ]
    constraintWeights [1,1,1,1,1]
   }
   "Constraint - dyn_Cloth_volume_l_A (PPPP)"!(
    ComponentPose=@n7
    Alpha=1
    ConstraintWeights_0=0.2372
    ConstraintWeights_1:float=0.4317
    ConstraintWeights_2:float=0.0623
    ConstraintWeights_3:float=0.2688
   ) -> (Pose) {
    alphaBoolBlend {blendOption=Linear}
    boneToModify {boneName=dyn_Cloth_volume_l_A}
    constraintSetup [
     {targetBone {boneName=pelvis}}
     {targetBone {boneName=thigh_twist_01_l}}
     {targetBone {boneName=thigh_l}}
     {targetBone {boneName=deform_groin_l}}
    ]
    constraintWeights [1,1,1,1]
   }
   Control Rig!(
    Source
    Alpha
    crouching:double=Get crouch() -> (crouch:double)
    crouch strafe r:double=Get strafe l() -> (strafe l:double)
    crouch strafe l:double=Get strafe r() -> (strafe r:double)
    moving 2d:double=Get Moving 2D() -> (Moving 2D:double)
    emote:double=Get emote() -> (emote:double)
   ) -> (Pose) {
    alphaBoolBlend {blendOption=Linear}
    controlRigAssetReference {blueprintRigClass=&/Pencil_SkirtLong_MedB_Lbdy_CtrlRig.Pencil_SkirtLong_MedB_Lbdy_CtrlRig_C}
   }
   n8 = Get bIsParachuteOpened()
   n9 = "Transform (Modify) Bone - Bone: dyn_Cloth_volume_l_B"!(ComponentPose, bAlphaBoolEnabled:bool=@n8, Translation="0,3,0", Rotation="0,5,0", Scale="1,1,1") -> (Pose) {
    alphaInputType=Bool
    rotationMode=BMM_Additive
    rotationSpace=BCS_BoneSpace
    translationMode=BMM_Additive
    translationSpace=BCS_BoneSpace
    alphaBoolBlend {blendInTime=0.3, blendOption=Linear, blendOutTime=0.3}
    boneToModify {boneName=dyn_Cloth_volume_l_B}
    rotation {yaw=5}
    translation {y=3}
   }
   "Transform (Modify) Bone - Bone: dyn_Cloth_volume_r_B"!(ComponentPose=@n9, bAlphaBoolEnabled:bool=@n8, Translation="0,3,0", Rotation="0,5,0", Scale="1,1,1") -> (Pose) {
    alphaInputType=Bool
    rotationMode=BMM_Additive
    rotationSpace=BCS_BoneSpace
    translationMode=BMM_Additive
    translationSpace=BCS_BoneSpace
    alphaBoolBlend {blendInTime=0.3, blendOption=Linear, blendOutTime=0.3}
    alphaScaleBiasClamp {bInterpResult=true, interpSpeedIncreasing=30}
    boneToModify {boneName=dyn_Cloth_volume_r_B}
    rotation {yaw=5}
    translation {y=3}
   }
   n10 = Control Rig!(Source=Copy Pose From Mesh() {bCopyCurves=true, bUseAttachedParent=true}, Alpha) -> (Pose) {
    alphaBoolBlend {blendOption=Linear}
    controlRigAssetReference {blueprintRigClass=&/Pencil_SkirtLong_MedB_Lbdy_DeformRig.Pencil_SkirtLong_MedB_Lbdy_DeformRig_C}
   }
   n11 = Control Rig(Source=Copy Pose From Mesh() {bCopyCurves=true, bUseAttachedParent=true}) {
    alphaBoolBlend {blendOption=Linear}
    controlRigAssetReference {blueprintRigClass=&Pencil_SkirtLong_MedB_Lbdy_DeformRig_C}
   }
   n12 = Get Blazer cuff() -> (Blazer cuff:bool)
   n13 = "Transform (Modify) Bone - Bone: dyn_Floor_COL"(
    ComponentPose=Local To Component(LocalPose=Blend Poses by bool(BlendPose_0=@n10, BlendPose_1=@n11, bActiveValue=@n12) {blendType=Linear})
    Alpha=Get FloorCOL() -> (FloorCOL:double)
    Translation="0,0,-100"
    Rotation="0,0,0"
    Scale="1,1,1"
   ) {
    rotationMode=BMM_Additive
    rotationSpace=BCS_BoneSpace
    translationMode=BMM_Additive
    translationSpace=BCS_BoneSpace
    alphaBoolBlend {blendOption=Linear}
    alphaScaleBiasClamp {bInterpResult=true}
    boneToModify {boneName=dyn_Floor_COL}
    translation {z=-100}
   }
   n14 = "Transform (Modify) Bone - Bone: dyn_Floor_COL"!(ComponentPose=@n13, bAlphaBoolEnabled:bool=True, Translation="0,0,-100", Rotation="0,0,0", Scale="1,1,1") -> (Pose) {
    alphaInputType=Bool
    rotationMode=BMM_Additive
    rotationSpace=BCS_BoneSpace
    scaleMode=BMM_Additive
    scaleSpace=BCS_BoneSpace
    translationMode=BMM_Additive
    translationSpace=BCS_BoneSpace
    alphaBoolBlend {blendInTime=0.1, blendOption=Linear, blendOutTime=0.1}
    alphaScaleBiasClamp {bInterpResult=true, interpSpeedDecreasing=5, interpSpeedIncreasing=8}
    boneToModify {boneName=dyn_Floor_COL}
    translation {z=-100}
   }
   n15 = Get bIsSliding()
   n16 = "Transform (Modify) Bone - Bone: dyn_Cloth_volume_bk"!(ComponentPose=@n14, bAlphaBoolEnabled:bool=@n15, Translation="30,-5,0", Rotation="0,55,0", Scale="1,1,1") -> (Pose) {
    alphaInputType=Bool
    rotationMode=BMM_Additive
    rotationSpace=BCS_BoneSpace
    translationMode=BMM_Additive
    translationSpace=BCS_BoneSpace
    alphaBoolBlend {blendInTime=0.3, blendOption=Linear, blendOutTime=0.3}
    boneToModify {boneName=dyn_Cloth_volume_bk}
    rotation {yaw=55}
    translation=<30 -5>
   }
   n17 = "Transform (Modify) Bone - Bone: dyn_Cloth_volume_l_C"!(ComponentPose=@n16, bAlphaBoolEnabled:bool=@n15, Translation="30,-5,-5", Rotation="10,55,0", Scale="1,1,1") -> (Pose) {
    alphaInputType=Bool
    rotationMode=BMM_Additive
    rotationSpace=BCS_BoneSpace
    translationMode=BMM_Additive
    translationSpace=BCS_BoneSpace
    alphaBoolBlend {blendInTime=0.3, blendOption=Linear, blendOutTime=0.3}
    boneToModify {boneName=dyn_Cloth_volume_l_C}
    rotation {pitch=10, yaw=55}
    translation=<30 -5 -5>
   }
   n18 = "Transform (Modify) Bone - Bone: dyn_Cloth_volume_r_C"!(ComponentPose=@n17, bAlphaBoolEnabled:bool=@n15, Translation="30,-5,5", Rotation="-10,55,0", Scale="1,1,1") -> (Pose) {
    alphaInputType=Bool
    rotationMode=BMM_Additive
    rotationSpace=BCS_BoneSpace
    translationMode=BMM_Additive
    translationSpace=BCS_BoneSpace
    alphaBoolBlend {blendInTime=0.3, blendOption=Linear, blendOutTime=0.3}
    boneToModify {boneName=dyn_Cloth_volume_r_C}
    rotation {pitch=-10, yaw=55}
    translation=<30 -5 5>
   }
   n19 = Get skirt fix() -> (skirt fix:bool)
   n20 = "Transform (Modify) Bone - Bone: dyn_Cloth_volume_l_B"(ComponentPose=@n18, bAlphaBoolEnabled=@n19, Translation="0,-2,0", Rotation="0,0,0", Scale="1,1,1") {
    alphaInputType=Bool
    rotationMode=BMM_Additive
    rotationSpace=BCS_BoneSpace
    translationMode=BMM_Additive
    translationSpace=BCS_BoneSpace
    alphaBoolBlend {blendInTime=0.3, blendOption=Linear, blendOutTime=0.3}
    boneToModify {boneName=dyn_Cloth_volume_l_B}
    translation {y=-2}
   }
   n21 = "Transform (Modify) Bone - Bone: dyn_Cloth_volume_r_B"(ComponentPose=@n20, bAlphaBoolEnabled=@n19, Translation="0,-3,0", Rotation="0,-5,0", Scale="1,1,1") {
    alphaInputType=Bool
    rotationMode=BMM_Additive
    rotationSpace=BCS_BoneSpace
    translationMode=BMM_Additive
    translationSpace=BCS_BoneSpace
    alphaBoolBlend {blendInTime=0.3, blendOption=Linear, blendOutTime=0.3}
    alphaScaleBiasClamp {bInterpResult=true, interpSpeedIncreasing=30}
    boneToModify {boneName=dyn_Cloth_volume_r_B}
    rotation {yaw=-5}
    translation {y=-3}
   }
   n22 = Get bIsMoving2D()
   n23 = "Transform (Modify) Bone - Bone: dyn_Cloth_volume_l_C02"!(ComponentPose=@n21, bAlphaBoolEnabled:bool=@n22, Translation="0,0,0", Rotation="0,30,0", Scale="1,1,1") -> (Pose) {
    alphaInputType=Bool
    rotationMode=BMM_Additive
    rotationSpace=BCS_BoneSpace
    translationMode=BMM_Additive
    translationSpace=BCS_BoneSpace
    alphaBoolBlend {blendInTime=0.1, blendOption=Linear, blendOutTime=0.3}
    boneToModify {boneName=dyn_Cloth_volume_l_C02}
    rotation {yaw=30}
   }
   n24 = "Transform (Modify) Bone - Bone: dyn_Cloth_volume_r_C02"!(ComponentPose=@n23, bAlphaBoolEnabled:bool=@n22, Translation="0,0,0", Rotation="0,30,0", Scale="1,1,1") -> (Pose) {
    alphaInputType=Bool
    rotationMode=BMM_Additive
    rotationSpace=BCS_BoneSpace
    translationMode=BMM_Additive
    translationSpace=BCS_BoneSpace
    alphaBoolBlend {blendInTime=0.1, blendOption=Linear, blendOutTime=0.3}
    boneToModify {boneName=dyn_Cloth_volume_r_C02}
    rotation {yaw=30}
   }
   n25 = "Transform (Modify) Bone - Bone: dyn_Cloth_volume_bk02"!(ComponentPose=@n24, bAlphaBoolEnabled:bool=@n22, Translation="0,0,0", Rotation="0,30,0", Scale="1,1,1") -> (Pose) {
    alphaInputType=Bool
    rotationMode=BMM_Additive
    rotationSpace=BCS_BoneSpace
    translationMode=BMM_Additive
    translationSpace=BCS_BoneSpace
    alphaBoolBlend {blendInTime=0.1, blendOption=Linear, blendOutTime=0.3}
    boneToModify {boneName=dyn_Cloth_volume_bk02}
    rotation {yaw=30}
   }
   n26 = Get bIsTacticalSprinting()
   n27 = "Transform (Modify) Bone - Bone: dyn_Cloth_volume_l_C01"!(ComponentPose=@n25, bAlphaBoolEnabled:bool=@n26, Translation="0,0,0", Rotation="0,35,0", Scale="1,1,1") -> (Pose) {
    alphaInputType=Bool
    rotationMode=BMM_Additive
    rotationSpace=BCS_BoneSpace
    translationMode=BMM_Additive
    translationSpace=BCS_BoneSpace
    alphaBoolBlend {blendInTime=0.1, blendOption=Linear, blendOutTime=0.3}
    boneToModify {boneName=dyn_Cloth_volume_l_C01}
    rotation {yaw=35}
   }
   n28 = "Transform (Modify) Bone - Bone: dyn_Cloth_volume_r_C01"!(ComponentPose=@n27, bAlphaBoolEnabled:bool=@n26, Translation="0,0,0", Rotation="0,35,0", Scale="1,1,1") -> (Pose) {
    alphaInputType=Bool
    rotationMode=BMM_Additive
    rotationSpace=BCS_BoneSpace
    translationMode=BMM_Additive
    translationSpace=BCS_BoneSpace
    alphaBoolBlend {blendInTime=0.1, blendOption=Linear, blendOutTime=0.3}
    boneToModify {boneName=dyn_Cloth_volume_r_C01}
    rotation {yaw=35}
   }
   n29 = "Transform (Modify) Bone - Bone: dyn_Cloth_volume_bk01"!(ComponentPose=@n28, bAlphaBoolEnabled:bool=@n26, Translation="0,0,0", Rotation="0,35,0", Scale="1,1,1") -> (Pose) {
    alphaInputType=Bool
    rotationMode=BMM_Additive
    rotationSpace=BCS_BoneSpace
    translationMode=BMM_Additive
    translationSpace=BCS_BoneSpace
    alphaBoolBlend {blendInTime=0.1, blendOption=Linear, blendOutTime=0.3}
    boneToModify {boneName=dyn_Cloth_volume_bk01}
    rotation {yaw=35}
   }
   n30 = Get Frontend() -> (Frontend:double)
   n31 = "Transform (Modify) Bone - Bone: dyn_Cloth_volume_l_A"(ComponentPose=@n29, Alpha=@n30, Translation="0,0,-5", Rotation="5,0,0", Scale="1,1,1") {
    rotationMode=BMM_Additive
    rotationSpace=BCS_BoneSpace
    translationMode=BMM_Additive
    translationSpace=BCS_BoneSpace
    alphaBoolBlend {blendInTime=0.3, blendOption=Linear, blendOutTime=0.3}
    alphaScaleBiasClamp {bInterpResult=true}
    boneToModify {boneName=dyn_Cloth_volume_l_A}
    rotation {pitch=5}
    translation {z=-5}
   }
   n32 = "Transform (Modify) Bone - Bone: dyn_Cloth_volume_r_A"(ComponentPose=@n31, Alpha=@n30, Translation="0,0,0", Rotation="0,5,0", Scale="1,1,1") {
    rotationMode=BMM_Additive
    rotationSpace=BCS_BoneSpace
    translationMode=BMM_Additive
    translationSpace=BCS_BoneSpace
    alphaBoolBlend {blendInTime=0.3, blendOption=Linear, blendOutTime=0.3}
    alphaScaleBiasClamp {bInterpResult=true}
    boneToModify {boneName=dyn_Cloth_volume_r_A}
    rotation {yaw=5}
   }
   n33 = "Constraint - dyn_Cloth_volume_fn (PP)"!(ComponentPose=@n32, Alpha=@n30, ConstraintWeights_0=0.6, ConstraintWeights_1:float=0.4) -> (Pose) {
    alphaBoolBlend {blendOption=Linear}
    boneToModify {boneName=dyn_Cloth_volume_fn}
    constraintSetup [{targetBone {boneName=dyn_Cloth_volume_r_A}}, {targetBone {boneName=dyn_Cloth_volume_l_A}}]
    constraintWeights [1,1]
   }
   n34 = "Constraint - dyn_Cloth_volume_fn01 (PP)"!(ComponentPose=@n33, Alpha=@n30, ConstraintWeights_0=0.6, ConstraintWeights_1:float=0.4) -> (Pose) {
    alphaBoolBlend {blendOption=Linear}
    boneToModify {boneName=dyn_Cloth_volume_fn01}
    constraintSetup [{targetBone {boneName=dyn_Cloth_volume_r_A01}}, {targetBone {boneName=dyn_Cloth_volume_l_A01}}]
    constraintWeights [1,1]
   }
   n35 = "Constraint - dyn_Cloth_volume_fn02 (PP)"!(ComponentPose=@n34, Alpha=@n30, ConstraintWeights_0=0.6, ConstraintWeights_1:float=0.4) -> (Pose) {
    alphaBoolBlend {blendOption=Linear}
    boneToModify {boneName=dyn_Cloth_volume_fn02}
    constraintSetup [{targetBone {boneName=dyn_Cloth_volume_r_A02}}, {targetBone {boneName=dyn_Cloth_volume_l_A02}}]
    constraintWeights [1,1]
   }
   n36 = Get Grind() -> (Grind:bool)
   n37 = "Transform (Modify) Bone - Bone: dyn_Floor_COL"(ComponentPose=@n35, bAlphaBoolEnabled=@n36, Translation="0,0,-100", Rotation="0,0,0", Scale="1,1,1") {
    alphaInputType=Bool
    rotationMode=BMM_Additive
    rotationSpace=BCS_BoneSpace
    translationMode=BMM_Additive
    translationSpace=BCS_BoneSpace
    alphaBoolBlend {blendOption=Linear}
    alphaScaleBiasClamp {bInterpResult=true}
    boneToModify {boneName=dyn_Floor_COL}
    translation {z=-100}
   }
   n38 = "Transform (Modify) Bone - Bone: dyn_Cloth_volume_l_C"(ComponentPose=@n37, bAlphaBoolEnabled=@n36, Translation="0,-2,0", Rotation="0,25,0", Scale="1,1,1") {
    alphaInputType=Bool
    rotationMode=BMM_Additive
    rotationSpace=BCS_BoneSpace
    translationMode=BMM_Additive
    translationSpace=BCS_BoneSpace
    alphaBoolBlend {blendInTime=0.1, blendOption=Linear, blendOutTime=0.3}
    boneToModify {boneName=dyn_Cloth_volume_l_C}
    rotation {yaw=25}
    translation {y=-2}
   }
   n39 = "Transform (Modify) Bone - Bone: dyn_Cloth_volume_r_C"(ComponentPose=@n38, bAlphaBoolEnabled=@n36, Translation="0,-2,0", Rotation="0,25,0", Scale="1,1,1") {
    alphaInputType=Bool
    rotationMode=BMM_Additive
    rotationSpace=BCS_BoneSpace
    translationMode=BMM_Additive
    translationSpace=BCS_BoneSpace
    alphaBoolBlend {blendInTime=0.1, blendOption=Linear, blendOutTime=0.3}
    boneToModify {boneName=dyn_Cloth_volume_r_C}
    rotation {yaw=25}
    translation {y=-2}
   }
   n40 = "Transform (Modify) Bone - Bone: dyn_Cloth_volume_bk"(ComponentPose=@n39, bAlphaBoolEnabled=@n36, Translation="0,-2,0", Rotation="0,25,0", Scale="1,1,1") {
    alphaInputType=Bool
    rotationMode=BMM_Additive
    rotationSpace=BCS_BoneSpace
    translationMode=BMM_Additive
    translationSpace=BCS_BoneSpace
    alphaBoolBlend {blendInTime=0.1, blendOption=Linear, blendOutTime=0.3}
    boneToModify {boneName=dyn_Cloth_volume_bk}
    rotation {yaw=25}
    translation {y=-2}
   }
   n41 = RigidBody!(ComponentPose=@n40, bAlphaBoolEnabled:bool, OverrideWorldGravity:Vector=Get GravityOverride(), ExternalForce) -> (Pose) {
    alphaInputType=Bool
    bClampLinearTranslationLimitToRefPose=true
    bOverrideWorldGravity=true
    overridePhysicsAsset=&/Pencil_SkirtLong_MedB_Lbdy_Physics
    simulationSpace=BaseBoneSpace
    alphaBoolBlend {blendInTime=0.1, blendOption=Linear, blendOutTime=0.1}
    baseBoneRef {boneName=pelvis}
    componentAppliedLinearAccClamp=<3000 3000 3000>
    componentLinearAccScale=<0.3 0.3 0.5>
    componentLinearVelScale=<0.3 0.3 0.5>
    simSpaceSettings {
     maxAngularAcceleration=1000
     maxAngularVelocity=3
     maxLinearAcceleration=700
     maxLinearVelocity=500
     velocityScaleZ=0
     worldAlpha=1
     externalLinearDragV=<0.2 0.2>
    }
   }
   n42 = Get strafe l()
   n43 = "Transform (Modify) Bone - Bone: dyn_Cloth_volume_l_B02"(ComponentPose=@n41, Alpha=@n42, Translation="0,0,0", Rotation="0,0,0", Scale="1,1,1") {
    rotationMode=BMM_Additive
    rotationSpace=BCS_BoneSpace
    translationMode=BMM_Additive
    translationSpace=BCS_BoneSpace
    alphaBoolBlend {blendInTime=0.3, blendOption=Linear, blendOutTime=0.3}
    alphaScaleBiasClamp {bInterpResult=true, interpSpeedIncreasing=30}
    boneToModify {boneName=dyn_Cloth_volume_l_B02}
   }
   n44 = "Transform (Modify) Bone - Bone: dyn_Cloth_volume_l_A02"(ComponentPose=@n43, Alpha=@n42, Translation="0,0,0", Rotation="-10,15,-0.001", Scale="1,1,1") {
    rotationMode=BMM_Additive
    rotationSpace=BCS_BoneSpace
    translationMode=BMM_Additive
    translationSpace=BCS_BoneSpace
    alphaBoolBlend {blendInTime=0.3, blendOption=Linear, blendOutTime=0.3}
    alphaScaleBiasClamp {bInterpResult=true, interpSpeedIncreasing=30}
    boneToModify {boneName=dyn_Cloth_volume_l_A02}
    rotation=<rot -10 15 -0.001>
   }
   n45 = "Transform (Modify) Bone - Bone: dyn_Cloth_volume_fn02"(ComponentPose=@n44, Alpha=@n42, Translation="0,0,0", Rotation="-15,15,0", Scale="1,1,1") {
    rotationMode=BMM_Additive
    rotationSpace=BCS_BoneSpace
    translationMode=BMM_Additive
    translationSpace=BCS_BoneSpace
    alphaBoolBlend {blendInTime=0.3, blendOption=Linear, blendOutTime=0.3}
    alphaScaleBiasClamp {bInterpResult=true, interpSpeedIncreasing=30}
    boneToModify {boneName=dyn_Cloth_volume_fn02}
    rotation {pitch=-15, yaw=15}
   }
   n46 = "Transform (Modify) Bone - Bone: dyn_Cloth_volume_r_C02"!(ComponentPose=@n45, Alpha:float=@n42, Translation="0,0,0", Rotation="-5,5,0", Scale="1,1,1") -> (Pose) {
    rotationMode=BMM_Additive
    rotationSpace=BCS_BoneSpace
    translationMode=BMM_Additive
    translationSpace=BCS_BoneSpace
    alphaBoolBlend {blendInTime=0.3, blendOption=Linear, blendOutTime=0.3}
    alphaScaleBiasClamp {bInterpResult=true, interpSpeedIncreasing=30}
    boneToModify {boneName=dyn_Cloth_volume_r_C02}
    rotation {pitch=-5, yaw=5}
   }
   n47 = "Transform (Modify) Bone - Bone: dyn_Cloth_volume_r_B01"(ComponentPose=@n46, Alpha=@n42, Translation="0,0,0", Rotation="5,5,0", Scale="1,1,1") {
    rotationMode=BMM_Additive
    rotationSpace=BCS_BoneSpace
    translationMode=BMM_Additive
    translationSpace=BCS_BoneSpace
    alphaBoolBlend {blendInTime=0.3, blendOption=Linear, blendOutTime=0.3}
    alphaScaleBiasClamp {bInterpResult=true, interpSpeedIncreasing=30}
    boneToModify {boneName=dyn_Cloth_volume_r_B01}
    rotation {pitch=5, yaw=5}
   }
   n48 = Get strafe r()
   n49 = "Transform (Modify) Bone - Bone: dyn_Cloth_volume_l_B01"(ComponentPose=@n47, Alpha=@n48, Translation="0,0,0", Rotation="-0.001,3,0", Scale="1,1,1") {
    rotationMode=BMM_Additive
    rotationSpace=BCS_BoneSpace
    translationMode=BMM_Additive
    translationSpace=BCS_BoneSpace
    alphaBoolBlend {blendInTime=0.3, blendOption=Linear, blendOutTime=0.3}
    alphaScaleBiasClamp {bInterpResult=true}
    boneToModify {boneName=dyn_Cloth_volume_l_B01}
    rotation {pitch=-0.001, yaw=3}
   }
   n50 = "Transform (Modify) Bone - Bone: dyn_Cloth_volume_l_C"!(ComponentPose=@n49, Alpha:float=@n48, Translation="0,0,0", Rotation="5,0,0", Scale="1,1,1") -> (Pose) {
    rotationMode=BMM_Additive
    rotationSpace=BCS_BoneSpace
    translationMode=BMM_Additive
    translationSpace=BCS_BoneSpace
    alphaBoolBlend {blendInTime=0.3, blendOption=Linear, blendOutTime=0.3}
    alphaScaleBiasClamp {bInterpResult=true}
    boneToModify {boneName=dyn_Cloth_volume_l_C}
    rotation {pitch=5}
   }
   n51 = "Transform (Modify) Bone - Bone: dyn_Cloth_volume_l_C01"!(ComponentPose=@n50, Alpha:float=@n48, Translation="0,0,0", Rotation="5,0,0", Scale="1,1,1") -> (Pose) {
    rotationMode=BMM_Additive
    rotationSpace=BCS_BoneSpace
    translationMode=BMM_Additive
    translationSpace=BCS_BoneSpace
    alphaBoolBlend {blendInTime=0.3, blendOption=Linear, blendOutTime=0.3}
    alphaScaleBiasClamp {bInterpResult=true}
    boneToModify {boneName=dyn_Cloth_volume_l_C01}
    rotation {pitch=5}
   }
   n52 = "Transform (Modify) Bone - Bone: dyn_Cloth_volume_r_B01"(ComponentPose=@n51, Alpha=@n48, Translation="0,0,0", Rotation="0,5,0", Scale="1,1,1") {
    rotationMode=BMM_Additive
    rotationSpace=BCS_BoneSpace
    translationMode=BMM_Additive
    translationSpace=BCS_BoneSpace
    alphaBoolBlend {blendInTime=0.3, blendOption=Linear, blendOutTime=0.3}
    alphaScaleBiasClamp {bInterpResult=true}
    boneToModify {boneName=dyn_Cloth_volume_r_B01}
    rotation {yaw=5}
   }
   n53 = Get crouch strafe r() -> (crouch strafe r:double)
   n54 = "Transform (Modify) Bone - Bone: dyn_Cloth_volume_l_B02"(ComponentPose=@n52, Alpha=@n53, Translation="0,0,0", Rotation="0,5,0", Scale="1,1,1") {
    rotationMode=BMM_Additive
    rotationSpace=BCS_BoneSpace
    translationMode=BMM_Additive
    translationSpace=BCS_BoneSpace
    alphaBoolBlend {blendInTime=0.3, blendOption=Linear, blendOutTime=0.3}
    alphaScaleBiasClamp {bInterpResult=true, interpSpeedIncreasing=30}
    boneToModify {boneName=dyn_Cloth_volume_l_B02}
    rotation {yaw=5}
   }
   n55 = "Transform (Modify) Bone - Bone: dyn_Cloth_volume_l_C01"(ComponentPose=@n54, Alpha=@n53, Translation="0,0,0", Rotation="15,10,0", Scale="1,1,1") {
    rotationMode=BMM_Additive
    rotationSpace=BCS_BoneSpace
    translationMode=BMM_Additive
    translationSpace=BCS_BoneSpace
    alphaBoolBlend {blendInTime=0.3, blendOption=Linear, blendOutTime=0.3}
    alphaScaleBiasClamp {bInterpResult=true, interpSpeedIncreasing=30}
    boneToModify {boneName=dyn_Cloth_volume_l_C01}
    rotation {pitch=15, yaw=10}
   }
   n56 = "Transform (Modify) Bone - Bone: dyn_Cloth_volume_r_B"(ComponentPose=@n55, bAlphaBoolEnabled=Get bIsInFrontEnd(), Translation="0,3,0", Rotation="0,-3,0", Scale="1,1,1") {
    alphaInputType=Bool
    rotationMode=BMM_Additive
    rotationSpace=BCS_BoneSpace
    translationMode=BMM_Additive
    translationSpace=BCS_BoneSpace
    alphaBoolBlend {blendOption=Linear}
    boneToModify {boneName=dyn_Cloth_volume_r_B}
    rotation {yaw=-3}
    translation {y=3}
   }
   n57 = Get bIsTacticalSprinting()
   n58 = "Transform (Modify) Bone - Bone: dyn_Cloth_volume_l_C"!(ComponentPose=@n56, bAlphaBoolEnabled:bool=@n57, Translation="0,-2,0", Rotation="0,5,0", Scale="1,1,1") -> (Pose) {
    alphaInputType=Bool
    rotationMode=BMM_Additive
    rotationSpace=BCS_BoneSpace
    translationMode=BMM_Additive
    translationSpace=BCS_BoneSpace
    alphaBoolBlend {blendInTime=0.1, blendOption=Linear, blendOutTime=0.3}
    boneToModify {boneName=dyn_Cloth_volume_l_C}
    rotation {yaw=5}
    translation {y=-2}
   }
   n59 = "Transform (Modify) Bone - Bone: dyn_Cloth_volume_r_C"(ComponentPose=@n58, bAlphaBoolEnabled=@n57, Translation="0,-2,0", Rotation="0,10,0", Scale="1,1,1") {
    alphaInputType=Bool
    rotationMode=BMM_Additive
    rotationSpace=BCS_BoneSpace
    translationMode=BMM_Additive
    translationSpace=BCS_BoneSpace
    alphaBoolBlend {blendInTime=0.1, blendOption=Linear, blendOutTime=0.3}
    boneToModify {boneName=dyn_Cloth_volume_r_C}
    rotation {yaw=10}
    translation {y=-2}
   }
   n60 = "Transform (Modify) Bone - Bone: dyn_Cloth_volume_bk"(ComponentPose=@n59, bAlphaBoolEnabled=@n57, Translation="0,-2,0", Rotation="0,7,0", Scale="1,1,1") {
    alphaInputType=Bool
    rotationMode=BMM_Additive
    rotationSpace=BCS_BoneSpace
    translationMode=BMM_Additive
    translationSpace=BCS_BoneSpace
    alphaBoolBlend {blendInTime=0.1, blendOption=Linear, blendOutTime=0.3}
    boneToModify {boneName=dyn_Cloth_volume_bk}
    rotation {yaw=7}
    translation {y=-2}
   }
   n61 = Get Zipline() -> (Zipline:bool)
   n62 = "Transform (Modify) Bone - Bone: dyn_Cloth_volume_l_C"(ComponentPose=@n60, bAlphaBoolEnabled=@n61, Translation="0,-2,0", Rotation="0,5,0", Scale="1,1,1") {
    alphaInputType=Bool
    rotationMode=BMM_Additive
    rotationSpace=BCS_BoneSpace
    translationMode=BMM_Additive
    translationSpace=BCS_BoneSpace
    alphaBoolBlend {blendInTime=0.1, blendOption=Linear, blendOutTime=0.3}
    boneToModify {boneName=dyn_Cloth_volume_l_C}
    rotation {yaw=5}
    translation {y=-2}
   }
   n63 = "Transform (Modify) Bone - Bone: dyn_Cloth_volume_r_C"(ComponentPose=@n62, bAlphaBoolEnabled=@n61, Translation="0,-2,0", Rotation="0,5,0", Scale="1,1,1") {
    alphaInputType=Bool
    rotationMode=BMM_Additive
    rotationSpace=BCS_BoneSpace
    translationMode=BMM_Additive
    translationSpace=BCS_BoneSpace
    alphaBoolBlend {blendInTime=0.1, blendOption=Linear, blendOutTime=0.3}
    boneToModify {boneName=dyn_Cloth_volume_r_C}
    rotation {yaw=5}
    translation {y=-2}
   }
   n64 = "Transform (Modify) Bone - Bone: dyn_Cloth_volume_bk"(ComponentPose=@n63, bAlphaBoolEnabled=@n61, Translation="0,-2,0", Rotation="0,5,0", Scale="1,1,1") {
    alphaInputType=Bool
    rotationMode=BMM_Additive
    rotationSpace=BCS_BoneSpace
    translationMode=BMM_Additive
    translationSpace=BCS_BoneSpace
    alphaBoolBlend {blendInTime=0.1, blendOption=Linear, blendOutTime=0.3}
    boneToModify {boneName=dyn_Cloth_volume_bk}
    rotation {yaw=5}
    translation {y=-2}
   }
   n65 = Get Grind()
   n66 = "Transform (Modify) Bone - Bone: dyn_Cloth_volume_l_C"(ComponentPose=@n64, bAlphaBoolEnabled=@n65, Translation="0,0,0", Rotation="0,15,0", Scale="1,1,1") {
    alphaInputType=Bool
    rotationMode=BMM_Additive
    rotationSpace=BCS_BoneSpace
    translationMode=BMM_Additive
    translationSpace=BCS_BoneSpace
    alphaBoolBlend {blendInTime=0.1, blendOption=Linear, blendOutTime=0.3}
    boneToModify {boneName=dyn_Cloth_volume_l_C}
    rotation {yaw=15}
   }
   n67 = "Transform (Modify) Bone - Bone: dyn_Cloth_volume_r_C"(ComponentPose=@n66, bAlphaBoolEnabled=@n65, Translation="0,0,0", Rotation="-15,15,0", Scale="1,1,1") {
    alphaInputType=Bool
    rotationMode=BMM_Additive
    rotationSpace=BCS_BoneSpace
    translationMode=BMM_Additive
    translationSpace=BCS_BoneSpace
    alphaBoolBlend {blendInTime=0.1, blendOption=Linear, blendOutTime=0.3}
    boneToModify {boneName=dyn_Cloth_volume_r_C}
    rotation {pitch=-15, yaw=15}
   }
   n68 = "Transform (Modify) Bone - Bone: dyn_Cloth_volume_bk"(ComponentPose=@n67, bAlphaBoolEnabled=@n65, Translation="0,0,0", Rotation="-5,15,0", Scale="1,1,1") {
    alphaInputType=Bool
    rotationMode=BMM_Additive
    rotationSpace=BCS_BoneSpace
    translationMode=BMM_Additive
    translationSpace=BCS_BoneSpace
    alphaBoolBlend {blendInTime=0.1, blendOption=Linear, blendOutTime=0.3}
    boneToModify {boneName=dyn_Cloth_volume_bk}
    rotation {pitch=-5, yaw=15}
   }
   n69 = "Transform (Modify) Bone - Bone: dyn_Cloth_volume_r_B"(ComponentPose=@n68, bAlphaBoolEnabled=@n65, Translation="0,0,0", Rotation="0,10,0", Scale="1,1,1") {
    alphaInputType=Bool
    rotationMode=BMM_Additive
    rotationSpace=BCS_BoneSpace
    translationMode=BMM_Additive
    translationSpace=BCS_BoneSpace
    alphaBoolBlend {blendInTime=0.1, blendOption=Linear, blendOutTime=0.3}
    boneToModify {boneName=dyn_Cloth_volume_r_B}
    rotation {yaw=10}
   }
   n70 = "Transform (Modify) Bone - Bone: dyn_Cloth_volume_l_B"(ComponentPose=@n69, bAlphaBoolEnabled=@n65, Translation="0,0,0", Rotation="0,10,0", Scale="1,1,1") {
    alphaInputType=Bool
    rotationMode=BMM_Additive
    rotationSpace=BCS_BoneSpace
    translationMode=BMM_Additive
    translationSpace=BCS_BoneSpace
    alphaBoolBlend {blendInTime=0.1, blendOption=Linear, blendOutTime=0.3}
    boneToModify {boneName=dyn_Cloth_volume_l_B}
    rotation {yaw=10}
   }
   n71 = Control Rig!(
    Source=Component To Local(ComponentPose=@n70)
    Alpha
    emote:double=Get emote()
    moving 2d:double=Get Moving 2D()
    pelvisDown:double=Get DetectedPelvisDown() -> (DetectedPelvisDown:double)
   ) -> (Pose) {
    alphaBoolBlend {blendOption=Linear}
    controlRigAssetReference {blueprintRigClass=&/Pencil_SkirtLong_MedB_Lbdy_post_CtrlRig.Pencil_SkirtLong_MedB_Lbdy_post_CtrlRig_C}
   }
   Output Pose(Result=@n71) {layerGroup=None}
  ]}
 EventGraph {nodes [
   Event BlueprintUpdateAnimation()
   > Set skirt fix(skirt fix:bool=true) -> (Output_Get:bool)
   > n1 = Sequence!(execute) -> (then_0, then_1, then_2:Exec, then_3:Exec)
   > n2 = Gravity Override State Function DEPRECATED(CharacterPartAnimInstance=Self-Reference(), GameTimeInSeconds=GetGameTimeInSeconds(), GravityOverrideParams=Get Gravity Override Params())
   > Set GravityOverride(GravityOverride=@n2)
   n3 = Get bIsMoving2D()
   > n4 = Branch(Condition=@n3)
   > Set Moving 2D(Moving 2D:double=1) -> (Output_Get:double)
   Set Moving 2D(execute=@n4.else, Moving 2D=0)
   n5 = PlayerMovement(execute=@n1.then_1, AnimInstance=Self-Reference(), XDotProductCoefficient=1, YDotProductCoefficient=1, ZDotProductCoefficient=1)
   n6 = MapRangeClamped(Value=@n5, InRangeA=0, InRangeB=1)
   > Set strafe r(strafe r:double=@n6) -> (Output_Get:double)
   > Set strafe l(strafe l:double=MapRangeClamped(Value=@n5, InRangeA=0, InRangeB=-1)) -> (Output_Get:double)
   > n7 = Branch(Condition=Get bIsCrouching())
   > Set crouch strafe r(crouch strafe r:double=@n6) -> (Output_Get:double)
   Set crouch strafe r(execute=@n7.else, crouch strafe r=0)
   n8 = Get bIsPlayingEmote()
   n9 = Branch(execute=@n1.then_2, Condition=AND Boolean(A=Get bIsInFrontEnd(), B=NOT Boolean(A=@n8)))
   > Set Frontend(Frontend:double=1) -> (Output_Get:double)
   Set Frontend(execute=@n9.else, Frontend=0)
   n10 = GetOwnerSkeletalMeshComponent()
   n11 = BreakVector(InVec=BreakTransform(InTransform=GetDeltaTransformFromRefPose(self=@n10, BoneName=pelvis, BaseName=root)))
   n12 = Branch(execute=@n1.then_3, Condition=AND Boolean(A="float < float"(A=@n11, B=-60), B=Get bIsPlayingEmote()))
   > Set DetectedPelvisDown(DetectedPelvisDown:double=1) -> (Output_Get:double)
   Set DetectedPelvisDown(execute=@n12.else, DetectedPelvisDown=0)
   Event OnCharacterCustomizationLoaded()
   > Sequence!(execute) -> (then_0, then_1)
   > n13 = Branch(Condition=CharacterItemDefinitionHasAnyMetaTag(Tags=Make Array("[0]"=(TagName="Cosmetics.MetaTags.Pencil.Upper.BlazerCuff"))))
   > Set Blazer cuff(Blazer cuff:bool=true) -> (Output_Get:bool)
   Set Blazer cuff(execute=@n13.else, Blazer cuff=false)
   n14 = Event OnBeginJumping()
   Set FloorCOL(execute=[@n14.then, @n18.then, @n19.then, @n22.then], FloorCOL:double=1) -> (Output_Get:double)
   n15 = Event OnEndJumping()
   Set FloorCOL(execute=[@n15.then, @n20.then, @n21.then, @n23.then], FloorCOL=0)
   Event OnBeginPlayingEmote() {is=title K2Node_Event}
   > Set emote(emote:double=1) -> (Output_Get:double)
   Event OnEndPlayingEmote() {is=title K2Node_Event}
   > Set emote(emote=0)
   Event OnEndCrouching()
   > Set crouch(crouch:double=0) -> (Output_Get:double)
   Event OnEnterGrindRail() {is=title K2Node_Event}
   > Set Grind(Grind:bool=true) -> (Output_Get:bool)
   Event OnExitGrindRail() {is=title K2Node_Event}
   > Set Grind(Grind=false)
   Event OnBeginCrouching()
   > Set crouch(crouch=1)
   n16 = Event BlueprintProcessAnimInputEvent()
   > Set Zipline(Zipline:bool=false) -> (Output_Get:bool)
   > n17 = Branch(Condition=IsNamed(self=@n16.AnimInputEvent, InEventName=aie.Player.Zipline.Entered))
   > Set Zipline(Zipline=true)
   Branch(execute=@n17.else, Condition=IsNamed(self=@n16.AnimInputEvent, InEventName=aie.Player.Zipline.Exited))
   > Set Zipline(Zipline=false)
   Event BlueprintInitializeAnimation()
   > Set Skel_Mesh(Skel_Mesh=GetOwnerSkeletalMeshComponent())
   n18 = Event OnBeginBaseballSliding() {is=title K2Node_Event}
   n19 = Event OnBeginKneeSliding() {is=title K2Node_Event}
   n20 = Event OnEndBaseballSliding() {is=title K2Node_Event}
   n21 = Event OnEndKneeSliding() {is=title K2Node_Event}
   n22 = Event OnBeginFalling()
   n23 = Event OnEndFalling()
   AND Boolean(A=@n3, B=NOT Boolean(A=Get bIsCrouching()))
  ]}
]
variables [
 Blazer cuff {type=boolean}
 crouch {type=number}
 crouch strafe r {type=number}
 DetectedPelvisDown {type=number}
 emote {type=number}
 FloorCOL {type=number}
 Frontend {type=number}
 Gravity Override Params {type=GravityOverrideParamsStruct}
 GravityOverride {type=Vector}
 Grind {type=boolean}
 Moving 2D {type=number}
 Skel_Mesh {type=SkeletalMeshComponent}
 skirt fix {type=boolean}
 strafe l {type=number}
 strafe r {type=number}
 Zipline {type=boolean}
]
